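# Physics asset for the face accessory part of the female medium white tennis outfit, simulating a dangling two-sided headband.
# Box bodies form a four-link left chain and a three-link right chain, each jointed link to link and rooted on the head.
# Capsules on head, neck, clavicles and upper spine give body shapes around the chains.
# Joints pair moderate swing limits, mostly looser toward the tips on one axis, with very tight twist.
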PhysicsAsset /BRCosmetics/Characters/Player/Female/Medium/Bodies/F_MED_Tennis_White/Meshes/Parts/F_MED_Tennis_White_FaceAcc_Physics
bodies {
 clavicle_l [{shapeType=Capsule, length=5, radius=7, center=<5.9504 0.3368 -2.4014>, rotation=<rot 87.4605 0 -4.3713>}]
 clavicle_r [{shapeType=Capsule, length=6.5, radius=6.5, center=<-5.6307 0.2516 2.3031>, rotation=<rot 90 0 -9.0462>}]
 dyn_headband_l_01 [{shapeType=Box, extentX=1.2954, extentY=0.6596, extentZ=1.6332, center=<0.8067 -0.0274 0.0033>, rotation=<rot 0 -92.0287 -89.7791>}]
 dyn_headband_l_02 [{shapeType=Box, extentX=1.7047, extentY=0.5, extentZ=2.0448, center=<1.0767 -0.0065 0.0001>, rotation=<rot 0 -0.3486 -85.2973>}]
 dyn_headband_l_03 [{shapeType=Box, extentX=2.5481, extentY=0.5, extentZ=2.3611, center=<1.5472 -0.0568 -0.0128>, rotation=<rot 0 -91.0954 -89.4836>}]
 dyn_headband_l_04 [{shapeType=Box, extentX=3.0213, extentY=0.5, extentZ=3.5672, center=<1.3196 -0.0133 -0.0826>, rotation=<rot 3.9609 -92.1213 -88.8519>}]
 dyn_headband_r_01 [{shapeType=Box, extentX=1.7469, extentY=0.8946, extentZ=2.1167, center=<1.2564 0.1062 -0.2066>, rotation=<rot 0 -95.748 -87.3411>}]
 dyn_headband_r_02 [{shapeType=Box, extentX=2.0489, extentY=0.63, extentZ=1.759, center=<0.902 -0.0197 -0.0388>, rotation=<rot -0.2483 -0.5662 -72.8592>}]
 dyn_headband_r_03 [{shapeType=Box, extentX=2.337, extentY=1.1903, extentZ=3.2677, center=<1.3614 -0.1879 0.0113>, rotation=<rot -5.4179 86.4578 -87.2136>}]
 head [{shapeType=Capsule, length=3.1908, radius=5.602, center=<-0.3634 0.6294 0>, rotation=<rot -90 0 30.0034>}]
 neck_01 [{shapeType=Capsule, length=3.5, radius=5, center=<3.3298 1.1847 0.0001>, rotation=<rot 90 0 0>}]
 spine_05 [{shapeType=Capsule, length=14.5, radius=14.755, center=<-2.3409 2.2964 0.0002>, rotation=<rot 0 -90 0>}]
}
constraints [
 {bone1Name=dyn_headband_l_02, bone2Name=dyn_headband_l_01, swing1LimitDegrees=20, swing2LimitDegrees=50, twistLimitDegrees=3}
 {bone1Name=dyn_headband_l_03, bone2Name=dyn_headband_l_02, swing1LimitDegrees=30, swing2LimitDegrees=50, twistLimitDegrees=3}
 {bone1Name=dyn_headband_l_04, bone2Name=dyn_headband_l_03, swing1LimitDegrees=40, swing2LimitDegrees=45, twistLimitDegrees=3}
 {bone1Name=dyn_headband_r_02, bone2Name=dyn_headband_r_01, swing1LimitDegrees=25, swing2LimitDegrees=45, twistLimitDegrees=3}
 {bone1Name=dyn_headband_r_03, bone2Name=dyn_headband_r_02, swing1LimitDegrees=35, swing2LimitDegrees=45, twistLimitDegrees=3}
 {bone1Name=dyn_headband_l_01, bone2Name=head, swing1LimitDegrees=45, swing2LimitDegrees=45, twistLimitDegrees=3}
 {bone1Name=dyn_headband_r_01, bone2Name=head, swing1LimitDegrees=45, swing2LimitDegrees=45, twistLimitDegrees=3}
]
