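# Physics asset for the ponytail on the female medium Tennis White character's FaceAcc mesh part.
# A chain of five small capsule bodies links through a driver body to the head, with swing and twist limits on each joint.
# Swing limits tighten along the upper-middle joints and widen toward the tip.
# Larger capsules on head, neck, clavicles and upper spine approximate the body around the hair and have no constraints.
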PhysicsAsset /BRCosmetics/Characters/Player/Female/Medium/Bodies/F_MED_Tennis_White/Meshes/Parts/F_MED_Tennis_White_FaceAcc_Ponytail_Physics
bodies {
 clavicle_l [{shapeType=Capsule, length=5, radius=7, center=<5.9504 0.3368 -2.4014>, rotation=<rot 87.4605 0 -4.3713>}]
 clavicle_r [{shapeType=Capsule, length=6.5, radius=6.5, center=<-5.6307 0.2516 2.3031>, rotation=<rot 90 0 -9.0462>}]
 driver_dyn_ponytail_01 [{shapeType=Capsule, length=0.505, radius=0.505, rotation=<rot 0 -90 0>}]
 dyn_ponytail_01 [{shapeType=Capsule, length=1.755, radius=1.505, center=<0.8909 0 0>, rotation=<rot 79.9997 0 0>}]
 dyn_ponytail_02 [{shapeType=Capsule, length=1.255, radius=1.255, center=<0.8543 0 0.0101>, rotation=<rot 79.9997 0 0>}]
 dyn_ponytail_03 [{shapeType=Capsule, length=1.005, radius=1.255, center=<1.1894 0 -0.03>, rotation=<rot 85.0002 0 0>}]
 dyn_ponytail_04 [{shapeType=Capsule, length=1.005, radius=1.255, center=<1.2699 0 0>, rotation=<rot 90 0 0>}]
 dyn_ponytail_05 [{shapeType=Capsule, length=1.505, radius=1.255, center=<1.623 0 0>, rotation=<rot 90 0 0>}]
 head [{shapeType=Capsule, length=3.1908, radius=5.602, center=<-0.3634 0.6294 0>, rotation=<rot -90 0 30.0034>}]
 neck_01 [{shapeType=Capsule, length=3.5, radius=5, center=<3.3298 1.1847 0.0001>, rotation=<rot 90 0 0>}]
 spine_05 [
  {shapeType=Capsule, length=14.5, radius=14.755, center=<-2.3409 2.2964 0.0002>, rotation=<rot 0 -90 0>}
  {shapeType=Capsule, length=7.5, radius=7.25, shapeName=spine_05_capsule, center=<8.2425 0.933 0>, rotation=<rot 0 -0.0001 10>}
 ]
}
constraints [
 {bone1Name=driver_dyn_ponytail_01, bone2Name=head, swing1LimitDegrees=45, swing2LimitDegrees=45, twistLimitDegrees=45}
 {bone1Name=dyn_ponytail_01, bone2Name=driver_dyn_ponytail_01, swing1LimitDegrees=45, swing2LimitDegrees=45, twistLimitDegrees=25}
 {bone1Name=dyn_ponytail_02, bone2Name=dyn_ponytail_01, swing1LimitDegrees=35, swing2LimitDegrees=35, twistLimitDegrees=15}
 {bone1Name=dyn_ponytail_03, bone2Name=dyn_ponytail_02, swing1LimitDegrees=35, swing2LimitDegrees=35, twistLimitDegrees=15}
 {bone1Name=dyn_ponytail_04, bone2Name=dyn_ponytail_03, swing1LimitDegrees=45, swing2LimitDegrees=45, twistLimitDegrees=10}
 {bone1Name=dyn_ponytail_05, bone2Name=dyn_ponytail_04, swing1LimitDegrees=60, swing2LimitDegrees=60, twistLimitDegrees=20}
]
