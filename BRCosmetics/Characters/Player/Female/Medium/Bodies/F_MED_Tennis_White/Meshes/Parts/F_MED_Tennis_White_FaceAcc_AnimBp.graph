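# Animation Blueprint for the female tennis-white character's face accessory, driving secondary motion on headband and ponytail bones.
# The copied parent-mesh pose gets additive headband offsets for targeting and head yaw, then is cached.
# With rigid body enabled, it runs AnimDynamics, skydive float/dive offsets and two RigidBody simulations; otherwise a ponytail LookAt applies.
# Skydiving, parachute (delayed) and targeting/crouch events reset dynamics to avoid stale simulation.
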AnimBlueprint /BRCosmetics/Characters/Player/Female/Medium/Bodies/F_MED_Tennis_White/Meshes/Parts/F_MED_Tennis_White_FaceAcc_AnimBp
target_skeleton=&/F_MED_Tennis_White_FaceAcc_Skeleton
graphs [
 AnimGraph {nodes [
   n1 = Anim Dynamics - Bone: driver_dyn_ponytail_01(ComponentPose=Local To Component(LocalPose=Use cached pose 'AnimPose'()), Alpha=0.5, ExternalForce="0,0,0") {
    angularDampingOverride=0.995
    angularSpringConstant=250
    bAngularSpring=true
    bGravityOverrideInSimSpace=true
    bOverrideAngularDamping=true
    gravityScale=0.33
    simulationSpace=RootRelative
    alphaBoolBlend {blendOption=Linear}
    boundBone {boneName=driver_dyn_ponytail_01}
    componentAppliedLinearAccClamp=<100 100 100>
    componentLinearAccScale=<1 1 1>
    componentLinearVelScale=<1 1 1>
    physicsBodyDefinitions [{
      sphereCollisionRadius=0
      boundBone {boneName=driver_dyn_ponytail_01}
      boxExtents=<20 20 20>
      constraintSetup {twistAxis=AxisZ, angularLimitsMax=<25 25 10>, angularLimitsMin=<-15 -25 -10>, angularTarget=<0 0 0>}
      localJointOffset=<0 0 -10>
     }]
   }
   n2 = Get FaceAcc_Alpha() -> (FaceAcc_Alpha:double)
   n3 = "Transform (Modify) Bone - Bone: dyn_headband_parent"(ComponentPose=@n1, Alpha=@n2, Translation="0,0,0", Rotation="-30,-50,0", Scale="1,1,1") {
    rotationMode=BMM_Additive
    translationMode=BMM_Additive
    alphaBoolBlend {blendOption=Linear}
    boneToModify {boneName=dyn_headband_parent}
    rotation {pitch=-30, yaw=-50}
   }
   n4 = "Transform (Modify) Bone - Bone: dyn_headband_l_02"(ComponentPose=@n3, Alpha=@n2, Translation="0,0,-1", Rotation="0,-25,0", Scale="1,1,1") {
    rotationMode=BMM_Additive
    translationMode=BMM_Additive
    alphaBoolBlend {blendOption=Linear}
    boneToModify {boneName=dyn_headband_l_02}
    rotation {yaw=-25}
    translation {z=-1}
   }
   n5 = "Transform (Modify) Bone - Bone: dyn_headband_r_02"(ComponentPose=@n4, Alpha=@n2, Translation="0,0,-2", Rotation="-20,-50,20", Scale="1,1,1") {
    rotationMode=BMM_Additive
    translationMode=BMM_Additive
    alphaBoolBlend {blendOption=Linear}
    boneToModify {boneName=dyn_headband_r_02}
    rotation=<rot -20 -50 20>
    translation {z=-2}
   }
   n6 = "Transform (Modify) Bone - Bone: parent_dyn_ponytail_01"(ComponentPose=@n5, Alpha=@n2, Translation="0,0,0", Rotation="0,-40,0", Scale="1,1,1") {
    rotationMode=BMM_Additive
    translationMode=BMM_Additive
    alphaBoolBlend {blendOption=Linear}
    boneToModify {boneName=parent_dyn_ponytail_01}
    rotation {yaw=-40}
   }
   n7 = "Transform (Modify) Bone - Bone: dyn_ponytail_03"(ComponentPose=@n6, Alpha=@n2, Translation="0,0,-1", Rotation="-40,0,0", Scale="1,1,1") {
    rotationMode=BMM_Additive
    translationMode=BMM_Additive
    alphaBoolBlend {blendOption=Linear}
    boneToModify {boneName=dyn_ponytail_03}
    rotation {pitch=-40}
    translation {z=-1}
   }
   n8 = RigidBody(ComponentPose=@n7, Alpha=0.7) {
    bClampLinearTranslationLimitToRefPose=true
    bOverrideWorldGravity=true
    overridePhysicsAsset=&/F_MED_Tennis_White_FaceAcc_Physics
    simulationSpace=BaseBoneSpace
    alphaBoolBlend {blendOption=Linear}
    baseBoneRef {boneName=spine_05}
    componentAppliedLinearAccClamp=<10 10 10>
    componentLinearAccScale=<1 1 1>
    componentLinearVelScale=<1 1 1>
    overrideWorldGravity {z=-980}
   }
   n9 = RigidBody(ComponentPose=@n8, Alpha=0.7) {
    bClampLinearTranslationLimitToRefPose=true
    bOverrideWorldGravity=true
    overridePhysicsAsset=&/F_MED_Tennis_White_FaceAcc_Ponytail_Physics
    simulationSpace=BaseBoneSpace
    alphaBoolBlend {blendOption=Linear}
    baseBoneRef {boneName=spine_05}
    componentAppliedLinearAccClamp=<10 10 10>
    componentLinearAccScale=<1 1 1>
    componentLinearVelScale=<1 1 1>
    overrideWorldGravity {z=-980}
   }
   n10 = Look At - Bone: dyn_ponytail_01(ComponentPose=Local To Component(LocalPose=Use cached pose 'AnimPose'()), Alpha=1) {
    alphaBoolBlend {blendOption=Linear}
    boneToModify {boneName=dyn_ponytail_01}
    lookAtLocation=<-15 -6.5>
    lookAtTarget {boneReference {boneName=spine_05}}
    lookAt_Axis {axis=<1 0>}
    lookUp_Axis {axis=<0 1>}
   }
   n11 = Blend Poses by bool(BlendPose_0=Component To Local(ComponentPose=@n9), BlendPose_1=Component To Local(ComponentPose=@n10), bActiveValue=Get bRigidBodyEnabled()) {blendType=Linear}
   Output Pose(Result=Component To Local(ComponentPose=Local To Component(LocalPose=@n11))) {layerGroup=None}
   n12 = SelectFloat(A=1, B=0, bPickA=Get bIsTargeting())
   n13 = "Transform (Modify) Bone - Bone: dyn_headband_l_04"(
    ComponentPose=Local To Component(LocalPose=Copy Pose From Mesh() {bCopyCurves=true, bUseAttachedParent=true})
    Alpha=@n12
    Translation="0,0,0"
    Rotation="20,0,0"
    Scale="1,1,1"
   ) {
    rotationMode=BMM_Additive
    translationMode=BMM_Additive
    alphaBoolBlend {blendOption=Linear}
    boneToModify {boneName=dyn_headband_l_04}
    rotation {pitch=20}
   }
   n14 = "Transform (Modify) Bone - Bone: dyn_headband_l_01"(ComponentPose=@n13, Alpha=@n12, Translation="0,0,1", Rotation="50,0,0", Scale="1,1,1") {
    rotationMode=BMM_Additive
    translationMode=BMM_Additive
    alphaBoolBlend {blendOption=Linear}
    boneToModify {boneName=dyn_headband_l_01}
    rotation {pitch=50}
    translation {z=1}
   }
   n15 = "Transform (Modify) Bone - Bone: dyn_headband_r_01"(ComponentPose=@n14, Alpha=@n12, Translation="0,0,0", Rotation="50,0,0", Scale="1,1,1") {
    rotationMode=BMM_Additive
    translationMode=BMM_Additive
    alphaBoolBlend {blendOption=Linear}
    boneToModify {boneName=dyn_headband_r_01}
    rotation {pitch=50}
   }
   n16 = Get Rotate_Headband() -> (Rotate_Headband:double)
   n17 = "Transform (Modify) Bone - Bone: dyn_headband_l_01"(ComponentPose=@n15, Alpha=@n16, Translation="0,0,0", Rotation="0,0,50", Scale="1,1,1") {
    rotationMode=BMM_Additive
    translationMode=BMM_Additive
    alphaBoolBlend {blendOption=Linear}
    boneToModify {boneName=dyn_headband_l_01}
    rotation {roll=50}
   }
   n18 = "Transform (Modify) Bone - Bone: dyn_headband_r_01"(ComponentPose=@n17, Alpha=@n16, Translation="0,0,0", Rotation="0,0,50", Scale="1,1,1") {
    rotationMode=BMM_Additive
    translationMode=BMM_Additive
    alphaBoolBlend {blendOption=Linear}
    boneToModify {boneName=dyn_headband_r_01}
    rotation {roll=50}
   }
   n19 = "Transform (Modify) Bone - Bone: dyn_headband_l_02"(ComponentPose=@n18, Alpha=@n16, Translation="0,0,0", Rotation="0,0,10", Scale="1,1,1") {
    rotationMode=BMM_Additive
    translationMode=BMM_Additive
    alphaBoolBlend {blendOption=Linear}
    boneToModify {boneName=dyn_headband_l_02}
    rotation {roll=10}
   }
   Save cached pose 'AnimPose'(Pose=Component To Local(ComponentPose=@n19)) {cacheName=AnimPose}
  ]}
 EventGraph {nodes [
   Event BlueprintUpdateAnimation()
   n1 = Get bIsSkydiveFloating()
   n2 = Get bIsSkydiveDiveMode()
   > Set FaceAcc_Alpha(FaceAcc_Alpha:double=SelectFloat(A=1, B=0, bPickA=OR Boolean(A=@n2, B=@n1))) -> (Output_Get:double)
   n3 = BreakTransform(InTransform=GetDeltaTransformFromRefPose(self=GetOwnerSkeletalMeshComponent(), BoneName=head, BaseName=spine_01))
   > Set Rotate_Headband(Rotate_Headband:double=MapRangeClamped(Value=BreakRotator(InRot=@n3.Rotation).Yaw, InRangeA=0, InRangeB=-49)) -> (Output_Get:double)
   Event OnParachuteOpened()
   > n4 = Delay(Duration=1.5)
   ResetDynamics(execute=[@n4.then, @n5.then, @n10.then, @n8.then], InTeleportType=TeleportPhysics)
   > ResetDynamics(InTeleportType=ResetPhysics)
   n5 = Event OnEndSkydiving()
   n6 = Event OnEndCrouching()
   n7 = Get bIsCrouching()
   n8 = Branch(execute=[@n6.then, @n9.then], Condition=AND Boolean(A=Get bIsTargeting(), B=NOT Boolean(A=@n7)))
   n9 = Event OnBeginTargeting() {is=title K2Node_Event}
   n10 = Event OnBeginSkydiving()
  ]}
]
variables [
 FaceAcc_Alpha {type=number}
 Rotate_Headband {type=number}
]
